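# Conversation database for Season 26 Snapshot Part 3 quest dialogue, covering Fish Thicc's Slap Juice side job and recruitment into Nolan Chance's heist.
# Entry points for Q03, Q07, Q09 and Q14 lead to linear speech chains, with Q07 branching across four NPCs.
# Opening lines auto-select when the quest is active, its objective incomplete and the NPC tag matches.
# Choices carry per-quest tags; Heartbreak Ranger and Love Ranger share Q07 objective 1.
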ConversationDatabase /BattlePassS26_Quests/Conversations/Snapshots/Part03/Conversation_S26_Snapshot_P03
entry_points [
 {tag=Conversation.Entry.Season26.Snapshot.P03.Q03, destinations [#1]}
 {tag=Conversation.Entry.Season26.Snapshot.P03.Q07, destinations [#2]}
 {tag=Conversation.Entry.Season26.Snapshot.P03.Q09, destinations [#3]}
 {tag=Conversation.Entry.Season26.Snapshot.P03.Q14, destinations [#4]}
]
nodes [
 ConversationEntryPointNode_4 {
  id=#4
  class=ConversationEntryPointNode
  outputs [#5]
  properties {entryTag {tagName=Conversation.Entry.Season26.Snapshot.P03.Q14}}
 }
 Speech_C_8 {
  id=#5
  class=Speech_C
  outputs [#6]
  properties {
   softTaskIcon=&T_UI_NPC_MainQuest
   generalConfig {defaultMessage="BRAH!!! I am so pumped we did it! Those coral buddies will be so swole, NOBODY will mess with 'em."}
  }
  sub_nodes [
   ConversationChoiceNode_46 {
    class=ConversationChoiceNode
    properties {
     defaultChoiceDisplayText=Talk to Fish Thicc
     choiceTags {gameplayTags [{tagName=Conversation.Choice.AutoSelectIfValid}, {tagName=Conversation.Choice.Category.MainQuestline}]}
    }
   }
   FortConversationRequirement_ControllerRequirement_88 {
    class=FortConversationRequirement_ControllerRequirement
    properties {
     nodeName=Require: Has Completed Quest
     requirement {
      class=FortControllerRequirement_HasCompletedQuest
      properties {bInverseCondition=true, questDefinition=/BattlePassS26_Quests/Items/QuestItems/Snapshot/Part03/quest_s26_snapshot_p03_q14}
     }
    }
   }
   FortConversationRequirement_ControllerRequirement_89 {
    class=FortConversationRequirement_ControllerRequirement
    properties {
     nodeName=Require: Is Quest Active
     requirement {
      class=FortControllerRequirement_IsQuestActive
      properties {questDefinition=&quest_s26_snapshot_p03_q14}
     }
    }
   }
   FortConversationRequirement_ControllerRequirement_90 {
    class=FortConversationRequirement_ControllerRequirement
    properties {
     nodeName=Require: Has Completed Quest Objective
     requirement {
      class=FortControllerRequirement_HasCompletedQuestObjective
      properties {
       bInverseCondition=true
       objectiveBackendName=quest_s26_snapshot_p03_q14_obj0
       questDefinition=&quest_s26_snapshot_p03_q14
      }
     }
    }
   }
   FortConversationRequirement_ControllerRequirement_91 {
    class=FortConversationRequirement_ControllerRequirement
    properties {
     nodeName=Require: NPC Has Conversation Tag
     participantId {tagName=Conversation.Participant.NPC}
     requirement {class=FortControllerRequirement_NPCHasConversationEntryTag, properties {queryTag {tagName=Conversation.Entry.FishThicc}}}
    }
   }
  ]
 }
 Speech_C_9 {
  id=#6
  class=Speech_C
  outputs []
  properties {generalConfig {defaultMessage=I guess a deal's a deal. Tell Nolan we're on. And honestly? I could kinda use a good heist...}}
  sub_nodes [ConversationChoiceNode_47 {
    class=ConversationChoiceNode
    properties {defaultChoiceDisplayText=Continue, choiceTags {gameplayTags [{tagName=Athena.Quests.S26.Conversations.P03.Choice.Q14}]}}
   }]
 }
 ConversationEntryPointNode_3 {
  id=#3
  class=ConversationEntryPointNode
  outputs [#7]
  properties {entryTag {tagName=Conversation.Entry.Season26.Snapshot.P03.Q09}}
 }
 Speech_C_6 {
  id=#7
  class=Speech_C
  outputs [#8]
  properties {
   softTaskIcon=&T_UI_NPC_MainQuest
   generalConfig {defaultMessage="You're still working Fish Thicc for me, right? Get him to agree to the heist yet?"}
  }
  sub_nodes [
   ConversationChoiceNode_44 {
    class=ConversationChoiceNode
    properties {
     defaultChoiceDisplayText=Talk to Nolan Chance
     choiceTags {gameplayTags [
       {tagName=Conversation.Choice.AutoSelectIfValid}
       {tagName=Conversation.Choice.Category.MainQuestline}
       {tagName=Athena.Quests.S26.Conversations.P03.Choice.Q09}
      ]}
    }
   }
   FortConversationRequirement_ControllerRequirement_84 {
    class=FortConversationRequirement_ControllerRequirement
    properties {
     nodeName=Require: Has Completed Quest
     requirement {
      class=FortControllerRequirement_HasCompletedQuest
      properties {bInverseCondition=true, questDefinition=/BattlePassS26_Quests/Items/QuestItems/Snapshot/Part03/quest_s26_snapshot_p03_q09}
     }
    }
   }
   FortConversationRequirement_ControllerRequirement_85 {
    class=FortConversationRequirement_ControllerRequirement
    properties {
     nodeName=Require: Is Quest Active
     requirement {
      class=FortControllerRequirement_IsQuestActive
      properties {questDefinition=&quest_s26_snapshot_p03_q09}
     }
    }
   }
   FortConversationRequirement_ControllerRequirement_86 {
    class=FortConversationRequirement_ControllerRequirement
    properties {
     nodeName=Require: Has Completed Quest Objective
     requirement {
      class=FortControllerRequirement_HasCompletedQuestObjective
      properties {
       bInverseCondition=true
       objectiveBackendName=quest_s26_snapshot_p03_q09_obj0
       questDefinition=&quest_s26_snapshot_p03_q09
      }
     }
    }
   }
   FortConversationRequirement_ControllerRequirement_87 {
    class=FortConversationRequirement_ControllerRequirement
    properties {
     nodeName=Require: NPC Has Conversation Tag
     participantId {tagName=Conversation.Participant.NPC}
     requirement {class=FortControllerRequirement_NPCHasConversationEntryTag, properties {queryTag {tagName=Conversation.Entry.MusketSlinger}}}
    }
   }
  ]
 }
 Speech_C_7 {
  id=#8
  class=Speech_C
  outputs [#9]
  properties {generalConfig {defaultMessage="*sigh* Alright, tell you what. I'll give you two a side job. You take that payout, build his little gym..."}}
  sub_nodes [ConversationChoiceNode_45 {
    class=ConversationChoiceNode
    properties {defaultChoiceDisplayText=Continue, choiceTags {gameplayTags [{tagName=Athena.Quests.S26.Conversations.P03.Choice.Q09}]}}
   }]
 }
 Speech_C_18 {
  id=#9
  class=Speech_C
  outputs []
  properties {generalConfig {defaultMessage="...and when he's done with that, he joins the heist. No backsies."}}
  sub_nodes [ConversationChoiceNode_8 {
    class=ConversationChoiceNode
    properties {defaultChoiceDisplayText=Continue, choiceTags {gameplayTags [{tagName=Athena.Quests.S26.Conversations.P03.Choice.Q09}]}}
   }]
 }
 ConversationEntryPointNode_2 {
  id=#2
  class=ConversationEntryPointNode
  outputs [#10, #11, #12, #13]
  properties {entryTag {tagName=Conversation.Entry.Season26.Snapshot.P03.Q07}}
 }
 Speech_C_3 {
  id=#13
  class=Speech_C
  outputs [#14]
  properties {
   softTaskIcon=&T_UI_NPC_MainQuest
   generalConfig {defaultMessage="YOU ARE AN EMISSARY OF THE SLAP JUICE FACTORY OWNED BY THAT CHARLATAN, KADO THORNE?"}
  }
  sub_nodes [
   ConversationChoiceNode_42 {
    class=ConversationChoiceNode
    properties {
     defaultChoiceDisplayText=Talk to Heartbreak Ranger
     choiceTags {gameplayTags [
       {tagName=Conversation.Choice.AutoSelectIfValid}
       {tagName=Conversation.Choice.Category.MainQuestline}
       {tagName=Athena.Quests.S26.Conversations.P03.Choice.Q07Obj1}
      ]}
    }
   }
   FortConversationRequirement_ControllerRequirement_80 {
    class=FortConversationRequirement_ControllerRequirement
    properties {
     nodeName=Require: Has Completed Quest
     requirement {
      class=FortControllerRequirement_HasCompletedQuest
      properties {bInverseCondition=true, questDefinition=/BattlePassS26_Quests/Items/QuestItems/Snapshot/Part03/quest_s26_snapshot_p03_q07}
     }
    }
   }
   FortConversationRequirement_ControllerRequirement_81 {
    class=FortConversationRequirement_ControllerRequirement
    properties {
     nodeName=Require: Is Quest Active
     requirement {
      class=FortControllerRequirement_IsQuestActive
      properties {questDefinition=&quest_s26_snapshot_p03_q07}
     }
    }
   }
   FortConversationRequirement_ControllerRequirement_82 {
    class=FortConversationRequirement_ControllerRequirement
    properties {
     nodeName=Require: Has Completed Quest Objective
     requirement {
      class=FortControllerRequirement_HasCompletedQuestObjective
      properties {
       bInverseCondition=true
       objectiveBackendName=quest_s26_snapshot_p03_q07_obj1
       questDefinition=&quest_s26_snapshot_p03_q07
      }
     }
    }
   }
   FortConversationRequirement_ControllerRequirement_83 {
    class=FortConversationRequirement_ControllerRequirement
    properties {
     nodeName=Require: NPC Has Conversation Tag
     participantId {tagName=Conversation.Participant.NPC}
     requirement {class=FortControllerRequirement_NPCHasConversationEntryTag, properties {queryTag {tagName=Conversation.Entry.CupidEvil}}}
    }
   }
  ]
 }
 Speech_C_4 {
  id=#14
  class=Speech_C
  outputs []
  properties {generalConfig {defaultMessage="DO NOT SEEK TO DOUSE THE FLAMES OF MY POWER WITH THIS \"SLAP JUICE\" ON BEHALF OF YOUR DARK MASTER."}}
  sub_nodes [ConversationChoiceNode_43 {
    class=ConversationChoiceNode
    properties {defaultChoiceDisplayText=Continue, choiceTags {gameplayTags [{tagName=Athena.Quests.S26.Conversations.P03.Choice.Q07Obj1}]}}
   }]
 }
 Speech_C_13 {
  id=#12
  class=Speech_C
  outputs [#15]
  properties {
   softTaskIcon=&T_UI_NPC_MainQuest
   generalConfig {defaultMessage="I don't think I want any Slap Juice. I'm made of stone, you see, so I can't taste it."}
  }
  sub_nodes [
   ConversationChoiceNode_1 {
    class=ConversationChoiceNode
    properties {
     defaultChoiceDisplayText=Talk to Love Ranger
     choiceTags {gameplayTags [
       {tagName=Conversation.Choice.AutoSelectIfValid}
       {tagName=Conversation.Choice.Category.MainQuestline}
       {tagName=Athena.Quests.S26.Conversations.P03.Choice.Q07Obj1}
      ]}
    }
   }
   FortConversationRequirement_ControllerRequirement_0 {
    class=FortConversationRequirement_ControllerRequirement
    properties {
     nodeName=Require: Has Completed Quest
     requirement {
      class=FortControllerRequirement_HasCompletedQuest
      properties {bInverseCondition=true, questDefinition=&quest_s26_snapshot_p03_q07}
     }
    }
   }
   FortConversationRequirement_ControllerRequirement_5 {
    class=FortConversationRequirement_ControllerRequirement
    properties {
     nodeName=Require: Is Quest Active
     requirement {
      class=FortControllerRequirement_IsQuestActive
      properties {questDefinition=&quest_s26_snapshot_p03_q07}
     }
    }
   }
   FortConversationRequirement_ControllerRequirement_6 {
    class=FortConversationRequirement_ControllerRequirement
    properties {
     nodeName=Require: Has Completed Quest Objective
     requirement {
      class=FortControllerRequirement_HasCompletedQuestObjective
      properties {
       bInverseCondition=true
       objectiveBackendName=quest_s26_snapshot_p03_q07_obj1
       questDefinition=&quest_s26_snapshot_p03_q07
      }
     }
    }
   }
   FortConversationRequirement_ControllerRequirement_7 {
    class=FortConversationRequirement_ControllerRequirement
    properties {
     nodeName=Require: NPC Has Conversation Tag
     participantId {tagName=Conversation.Participant.NPC}
     requirement {class=FortControllerRequirement_NPCHasConversationEntryTag, properties {queryTag {tagName=Conversation.Entry.LoveRanger}}}
    }
   }
  ]
 }
 Speech_C_14 {
  id=#15
  class=Speech_C
  outputs []
  properties {generalConfig {defaultMessage="But also, given the Slap Factory's recent acquisition by Kado Thorne... this is a bad look for you."}}
  sub_nodes [ConversationChoiceNode_2 {
    class=ConversationChoiceNode
    properties {defaultChoiceDisplayText=Continue, choiceTags {gameplayTags [{tagName=Athena.Quests.S26.Conversations.P03.Choice.Q07Obj1}]}}
   }]
 }
 Speech_C_12 {
  id=#11
  class=Speech_C
  outputs [#16]
  properties {
   softTaskIcon=&T_UI_NPC_MainQuest
   generalConfig {defaultMessage="Listen here, bub. I'm not interested in getting tangled up in Kado Thorne's web."}
  }
  sub_nodes [
   ConversationChoiceNode_40 {
    class=ConversationChoiceNode
    properties {
     defaultChoiceDisplayText=Talk to Renegade Shadow
     choiceTags {gameplayTags [
       {tagName=Conversation.Choice.AutoSelectIfValid}
       {tagName=Conversation.Choice.Category.MainQuestline}
       {tagName=Athena.Quests.S26.Conversations.P03.Choice.Q07Obj2}
      ]}
    }
   }
   FortConversationRequirement_ControllerRequirement_76 {
    class=FortConversationRequirement_ControllerRequirement
    properties {
     nodeName=Require: Has Completed Quest
     requirement {
      class=FortControllerRequirement_HasCompletedQuest
      properties {bInverseCondition=true, questDefinition=&quest_s26_snapshot_p03_q07}
     }
    }
   }
   FortConversationRequirement_ControllerRequirement_77 {
    class=FortConversationRequirement_ControllerRequirement
    properties {
     nodeName=Require: Is Quest Active
     requirement {
      class=FortControllerRequirement_IsQuestActive
      properties {questDefinition=&quest_s26_snapshot_p03_q07}
     }
    }
   }
   FortConversationRequirement_ControllerRequirement_78 {
    class=FortConversationRequirement_ControllerRequirement
    properties {
     nodeName=Require: Has Completed Quest Objective
     requirement {
      class=FortControllerRequirement_HasCompletedQuestObjective
      properties {
       bInverseCondition=true
       objectiveBackendName=quest_s26_snapshot_p03_q07_obj2
       questDefinition=&quest_s26_snapshot_p03_q07
      }
     }
    }
   }
   FortConversationRequirement_ControllerRequirement_79 {
    class=FortConversationRequirement_ControllerRequirement
    properties {
     nodeName=Require: NPC Has Conversation Tag
     participantId {tagName=Conversation.Participant.NPC}
     requirement {class=FortControllerRequirement_NPCHasConversationEntryTag, properties {queryTag {tagName=Conversation.Entry.RenegadeShadow}}}
    }
   }
  ]
 }
 Speech_C_11 {
  id=#16
  class=Speech_C
  outputs []
  properties {generalConfig {defaultMessage="Er, bats don't do webs. But still. If Slap's selling to Kado Thorne, I'm not interested."}}
  sub_nodes [ConversationChoiceNode_41 {
    class=ConversationChoiceNode
    properties {defaultChoiceDisplayText=Continue, choiceTags {gameplayTags [{tagName=Athena.Quests.S26.Conversations.P03.Choice.Q07Obj2}]}}
   }]
 }
 Speech_C_0 {
  id=#10
  class=Speech_C
  outputs [#17]
  properties {
   softTaskIcon=&T_UI_NPC_MainQuest
   generalConfig {defaultMessage="Slap Corp? As in, the Slap Corp that sold their factory to Kado Thorne?"}
  }
  sub_nodes [
   ConversationChoiceNode_0 {
    class=ConversationChoiceNode
    properties {
     defaultChoiceDisplayText=Talk to Diamond Diva
     choiceTags {gameplayTags [
       {tagName=Conversation.Choice.AutoSelectIfValid}
       {tagName=Conversation.Choice.Category.MainQuestline}
       {tagName=Athena.Quests.S26.Conversations.P03.Choice.Q07Obj0}
      ]}
    }
   }
   FortConversationRequirement_ControllerRequirement_1 {
    class=FortConversationRequirement_ControllerRequirement
    properties {
     nodeName=Require: Has Completed Quest
     requirement {
      class=FortControllerRequirement_HasCompletedQuest
      properties {bInverseCondition=true, questDefinition=&quest_s26_snapshot_p03_q07}
     }
    }
   }
   FortConversationRequirement_ControllerRequirement_2 {
    class=FortConversationRequirement_ControllerRequirement
    properties {
     nodeName=Require: Is Quest Active
     requirement {
      class=FortControllerRequirement_IsQuestActive
      properties {questDefinition=&quest_s26_snapshot_p03_q07}
     }
    }
   }
   FortConversationRequirement_ControllerRequirement_3 {
    class=FortConversationRequirement_ControllerRequirement
    properties {
     nodeName=Require: Has Completed Quest Objective
     requirement {
      class=FortControllerRequirement_HasCompletedQuestObjective
      properties {
       bInverseCondition=true
       objectiveBackendName=quest_s26_snapshot_p03_q07_obj0
       questDefinition=&quest_s26_snapshot_p03_q07
      }
     }
    }
   }
   FortConversationRequirement_ControllerRequirement_4 {
    class=FortConversationRequirement_ControllerRequirement
    properties {
     nodeName=Require: NPC Has Conversation Tag
     participantId {tagName=Conversation.Participant.NPC}
     requirement {class=FortControllerRequirement_NPCHasConversationEntryTag, properties {queryTag {tagName=Conversation.Entry.DiamondDiva}}}
    }
   }
  ]
 }
 Speech_C_5 {
  id=#17
  class=Speech_C
  outputs []
  properties {generalConfig {defaultMessage=No thank you. I'd rather get involved with angry wolves than that high-class hoarder.}}
  sub_nodes [ConversationChoiceNode_39 {
    class=ConversationChoiceNode
    properties {defaultChoiceDisplayText=Continue, choiceTags {gameplayTags [{tagName=Athena.Quests.S26.Conversations.P03.Choice.Q07Obj0}]}}
   }]
 }
 ConversationEntryPointNode_1 {
  id=#1
  class=ConversationEntryPointNode
  outputs [#18]
  properties {entryTag {tagName=Conversation.Entry.Season26.Snapshot.P03.Q03}}
 }
 Speech_C_1 {
  id=#18
  class=Speech_C
  outputs [#19]
  properties {
   softTaskIcon=&T_UI_NPC_MainQuest
   generalConfig {defaultMessage="BRAAAAAH! You're that Slap enthusiant people have told me about. I like you already."}
  }
  sub_nodes [
   ConversationChoiceNode_36 {
    class=ConversationChoiceNode
    properties {
     defaultChoiceDisplayText=Talk to Fish Thicc
     choiceTags {gameplayTags [
       {tagName=Conversation.Choice.AutoSelectIfValid}
       {tagName=Conversation.Choice.Category.MainQuestline}
       {tagName=Athena.Quests.S26.Conversations.P03.Choice.Q03}
      ]}
    }
   }
   FortConversationRequirement_ControllerRequirement_68 {
    class=FortConversationRequirement_ControllerRequirement
    properties {
     nodeName=Require: Has Completed Quest
     requirement {
      class=FortControllerRequirement_HasCompletedQuest
      properties {bInverseCondition=true, questDefinition=/BattlePassS26_Quests/Items/QuestItems/Snapshot/Part03/quest_s26_snapshot_p03_q03}
     }
    }
   }
   FortConversationRequirement_ControllerRequirement_69 {
    class=FortConversationRequirement_ControllerRequirement
    properties {
     nodeName=Require: Is Quest Active
     requirement {
      class=FortControllerRequirement_IsQuestActive
      properties {questDefinition=&quest_s26_snapshot_p03_q03}
     }
    }
   }
   FortConversationRequirement_ControllerRequirement_70 {
    class=FortConversationRequirement_ControllerRequirement
    properties {
     nodeName=Require: Has Completed Quest Objective
     requirement {
      class=FortControllerRequirement_HasCompletedQuestObjective
      properties {
       bInverseCondition=true
       objectiveBackendName=quest_s26_snapshot_p03_q03_obj0
       questDefinition=&quest_s26_snapshot_p03_q03
      }
     }
    }
   }
   FortConversationRequirement_ControllerRequirement_71 {
    class=FortConversationRequirement_ControllerRequirement
    properties {
     nodeName=Require: NPC Has Conversation Tag
     participantId {tagName=Conversation.Participant.NPC}
     requirement {class=FortControllerRequirement_NPCHasConversationEntryTag, properties {queryTag {tagName=Conversation.Entry.FishThicc}}}
    }
   }
  ]
 }
 Speech_C_2 {
  id=#19
  class=Speech_C
  outputs [#20]
  properties {generalConfig {defaultMessage="No no no, stop right there. I don't got time for any \"heists\" until I make a lil gym for the Coral Buddies."}}
  sub_nodes [ConversationChoiceNode_3 {
    class=ConversationChoiceNode
    properties {defaultChoiceDisplayText=Continue, choiceTags {gameplayTags [{tagName=Athena.Quests.S26.Conversations.P03.Choice.Q03}]}}
   }]
 }
 Speech_C_10 {
  id=#20
  class=Speech_C
  outputs [#21]
  properties {generalConfig {defaultMessage="My cousin said some dangerous types tried to swipe 'em a while back. NOT ON MY WATCH!"}}
  sub_nodes [ConversationChoiceNode_37 {
    class=ConversationChoiceNode
    properties {defaultChoiceDisplayText=Continue, choiceTags {gameplayTags [{tagName=Athena.Quests.S26.Conversations.P03.Choice.Q03}]}}
   }]
 }
 Speech_C_15 {
  id=#21
  class=Speech_C
  outputs []
  properties {generalConfig {defaultMessage="But I need dough. So you get to be the first employee of my new SLAP JUICE FRANCHISE, brah!"}}
  sub_nodes [ConversationChoiceNode_4 {
    class=ConversationChoiceNode
    properties {defaultChoiceDisplayText=Continue, choiceTags {gameplayTags [{tagName=Athena.Quests.S26.Conversations.P03.Choice.Q03}]}}
   }]
 }
]
